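# Animation Blueprint for the Teriyaki Fish Toon Skeleton backpack cosmetic, adding springy secondary motion to its idle animation.
# A single-entry random player feeds three chained AnimDynamics nodes on water dynamics bones, then an additive offset on the pack bone.
# The first two use angular springs with smoothed alpha from jumpReactivity, set each update lower while bIsJumping is true.
# The third runs at full alpha in bone-relative space with linear and angular springs.
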AnimBlueprint /BRCosmetics/Accessories/FORT_Backpacks/Backpack_M_MED_Teriyaki_Fish_Toon_Skeleton/Meshes/M_MED_Teriyaki_Fish_Toon_Skeleton_Pack_AnimBP
target_skeleton=&/M_MED_Teriyaki_Fish_Toon_Skeleton_Pack_Skeleton
graphs [
 AnimGraph {nodes [
   n1 = Random Sequence Player!() -> (Pose:Pose Link Structure) {
    is=title AnimGraphNode_RandomPlayer
    entries [{
      chanceToPlay=1
      maxLoopCount=0
      maxPlayRate=1
      minLoopCount=0
      minPlayRate=1
      sequence=/BRCosmetics/Accessories/FORT_Backpacks/Backpack_M_MED_Teriyaki_Fish_Toon_Skeleton/Animation/Teriyaki_Fish_Toon_Skeleton_Idle_Dead_BACKPACK
      blendIn {blendOption=Linear, blendTime=0.2, customCurve=null}
     }]
   }
   n2 = Get jumpReactivity() -> (jumpReactivity:double)
   n3 = Anim Dynamics - Bone: dyn_waterDynamics_1(ComponentPose=Local To Component(LocalPose=@n1), Alpha=@n2, ExternalForce="0,0,0") {
    angularSpringConstant=500
    bAngularSpring=true
    bGravityOverrideInSimSpace=true
    gravityScale=0
    alphaBoolBlend {blendOption=Linear}
    alphaScaleBiasClamp {bInterpResult=true, interpSpeedDecreasing=20, interpSpeedIncreasing=20, scale=0.82}
    boundBone {boneName=dyn_waterDynamics_1}
    componentLinearAccScale=<0.9 0.9 0.9>
    physicsBodyDefinitions [{
      sphereCollisionRadius=0
      boundBone {boneName=dyn_waterDynamics_1}
      boxExtents=<5 5 5>
      constraintSetup {angularLimitsMax=<15 30 30>, angularLimitsMin=<-15 -2 -30>, angularTarget=<1 0 0>}
      localJointOffset=<0 -0.2 0.5>
     }]
   }
   n4 = Anim Dynamics - Bone: dyn_waterDynamics_2(ComponentPose=@n3, Alpha=@n2, ExternalForce="0,0,0") {
    angularSpringConstant=500
    bAngularSpring=true
    bGravityOverrideInSimSpace=true
    gravityScale=0
    alphaBoolBlend {blendOption=Linear}
    alphaScaleBiasClamp {bInterpResult=true, interpSpeedDecreasing=20, interpSpeedIncreasing=20, scale=0.82}
    boundBone {boneName=dyn_waterDynamics_2}
    componentLinearAccScale=<0.9 0.9 0.9>
    physicsBodyDefinitions [{
      sphereCollisionRadius=0
      boundBone {boneName=dyn_waterDynamics_2}
      boxExtents=<5 5 5>
      constraintSetup {angularLimitsMax=<15 40 30>, angularLimitsMin=<-15 -2 -30>, angularTarget=<1 0 0>}
      localJointOffset=<0 0.2 -0.5>
     }]
   }
   n5 = Anim Dynamics - Bone: dyn_waterDynamics_3(ComponentPose=@n4, Alpha=1, ExternalForce="0,0,0") {
    angularSpringConstant=50
    bAngularSpring=true
    bGravityOverrideInSimSpace=true
    bLinearSpring=true
    linearSpringConstant=200
    simulationSpace=BoneRelative
    alphaBoolBlend {blendOption=Linear}
    boundBone {boneName=dyn_waterDynamics_3}
    componentLinearAccScale=<1.5 1.5 1.15>
    physicsBodyDefinitions [{
      sphereCollisionRadius=0
      boundBone {boneName=dyn_waterDynamics_3}
      boxExtents=<5 5 5>
      constraintSetup {angularLimitsMax=<10 10 10>, angularLimitsMin=<-10 -10 -10>, angularTarget=<1 0 0>, linearAxesMax=<5 0 0>}
      localJointOffset=<0 0 0>
     }]
   }
   n6 = "Transform (Modify) Bone - Bone: pack"(ComponentPose=@n5, Translation="0,-0.8,1", Rotation="0,0,0") {translationMode=BMM_Additive, alphaBoolBlend {blendOption=Linear}, boneToModify {boneName=pack}, translation {y=-0.8, z=1}}
   Output Pose(Result=Component To Local(ComponentPose=@n6)) {layerGroup=None}
  ]}
 EventGraph {nodes [
   Event BlueprintUpdateAnimation()
   > Set jumpReactivity(jumpReactivity:double=SelectFloat(A=0.4, B=0.6, bPickA=Get bIsJumping())) -> (Output_Get:double)
   TryGetPawnOwner()
  ]}
]
variables [jumpReactivity {type=number}]
